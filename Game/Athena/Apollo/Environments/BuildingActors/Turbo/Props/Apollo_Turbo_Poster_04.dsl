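; Interactable poster prop that players deface with a marker to send a Donut poster quest stat event.
; On BeginPlay the authority always enables interaction, while remotes wait for quests to load and enable it only if the quest is present and uncompleted.
; Interaction sets IsDefaced, blocking reuse, then swaps to the defaced material and plays a sound.
; Bounce animation updates reapply that material; OnRep_IsDefaced is empty.
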(event EventBeginPlay
  (bind _returnvalue (Fortnite|GetFirstLocalFortPlayerController true))
  (Networking|SwitchHasAuthority
    (:Authority
      (Variables|Default|SetCanInteract true))
    (:Remote
      (bind (_owningquestmanager _bfoundallquestsonprofile) (Quests|CurrentQuestsReadyAsync (Quests|GetQuestManager _returnvalue "Invalid") (Utilities|Array|MakeArray (Variables|Default|GetQuestToCheck)))
        (:OnCurrentQuestsReady
          (bind (_playercontrollerout _completedquest) (Quest|HasPlayerCompletedQuest _returnvalue (Variables|Default|GetQuestToCheck) (Variables|Default|GetQuestBackendName)))
          (Variables|Default|SetCanInteract (and _bfoundallquestsonprofile (not _completedquest))))))))

(event Interaction|EventBlueprintOnInteract
  (params (InteractingPawn :type Fort%20Pawn%20Object%20Reference) (InteractionBeingAttempted :type EInteractionBeingAttempted%20Enum :default "FirstInteraction"))
  (Quest|SendComplexCustomStatEvent (Quests|GetQuestManager (Fortnite|GetFortPlayerControllerfromActor InteractingPawn) "Invalid") 0 "(GameplayTags=)" "(GameplayTags=((TagName=\"Athena.Quests.Donut.Poster\")))" 1)
  (Variables|Default|SetIsDefaced true)
  (CallFunction|PosterCosmeticEffects))

(event Custom|PosterCosmeticEffects
  ;[1 1 1] Play marker sound, and swap material to defaced version
  (Rendering|Material|SetMaterial (Variables|Debug|GetStaticMeshComponent) 0 (Variables|Default|GetDefacedMaterial))
  (Audio|PlaySoundatLocation (Variables|Default|GetMarkerSound) (Transformation|GetActorLocation)))

(event EventOnBounceAnimationUpdate
  (params (Data :type Fort%20Bounce%20Data%20Structure%20%28by%20ref%29))
  ;[1 1 1] Fix material if this gets bounced by the thing it is attached to
  (if (Variables|Default|GetIsDefaced)
    (Rendering|Material|SetMaterial (Variables|Debug|GetStaticMeshComponent) 0 (Variables|Default|GetDefacedMaterial))))

(fn BlueprintCanInteract
  (params (InteractingPawn :type Fort%20Pawn%20Object%20Reference) (InteractionBeingAttempted :type EInteractionBeingAttempted%20Enum :default "FirstInteraction") (InteractionType :type TInteractionType%20Enum :default "IT_NoInteraction"))
  (return (and (Variables|Default|GetCanInteract) (not (Variables|Default|GetIsDefaced)))))

(fn BlueprintGetInteractionString
  (params (InteractingPawn :type Fort%20Pawn%20Object%20Reference) (InteractionBeingAttempted :type EInteractionBeingAttempted%20Enum :default "FirstInteraction"))
  (return (Variables|Default|GetInteractString)))

(fn OnRep_IsDefaced)

(fn ConstructionScript)
